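; BP_LedgeToggleController is a Blast Berry map actor that sets ledge launching data on the building wall subsystem.
; On BeginPlay it fetches that subsystem and waits asynchronously for playlist data to be ready.
; Once ready, it checks the subsystem is valid and passes itself and its bShouldSpawnLedge variable to UpdateLedgeLaunchingData.
; The overlap, tick and construction script are empty.
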(event EventBeginPlay
  (bind _returnvalue (WorldSubsystems|GetBuildingWallSubsystem))
  (Class|FortAsyncActionCurrentPlaylistReady|PlaylistDataReadyAsync
    (:OnReady
      (Utilities|IsValid _returnvalue
        (:Is%20Valid
          (Class|BuildingWallSubsystem|UpdateLedgeLaunchingData _returnvalue self (Variables|Default|GetbShouldSpawnLedge)))))))

(event Collision|EventActorBeginOverlap
  (params (OtherActor :type Actor%20Object%20Reference)))

(event EventTick
  (params (DeltaSeconds :type Float%20%28single-precision%29)))

(fn ConstructionScript)
