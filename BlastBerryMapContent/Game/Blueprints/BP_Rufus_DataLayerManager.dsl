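; Rufus data layer manager Blueprint that switches World Partition data layers to match the map's update week and scorch mark state.
; SetRufusWeekDataLayers maps a map index to a week string, stores it as CurrentWeek, then activates enabled and unloads disabled layers from matching calendar list entries.
; HandleScorchMarkDataLayers uses an index to set the scorch mark layer Unloaded (0) or Activated (1).
; No callers of either function appear here.
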(fn HandleScorchMarkDataLayers
  (params (Scorch%20Mark%20Index :type Integer))
  (bind _returnvalue (WorldPartition|GetDataLayerManager))
  (DataLayers|SetDataLayerInstanceRuntimeState _returnvalue (DataLayers|GetDataLayerInstancefromAsset _returnvalue (Variables|Default|GetScorch%20Mark%20Data%20Layer)) (Utilities|Select :Index Scorch%20Mark%20Index :Option%200 "Unloaded" :Option%201 "Activated") false))

(fn SetRufusWeekDataLayers
  (params (MapIndex :type Integer))
  (Variables|Default|SetCurrentWeek (Utilities|Array|Get(acopy) (Variables|Default|GetMapUpdateWeekArray) MapIndex))
  (for _array_element (Variables|Default|GetCalendarDataLayerList)
    (if (Utilities|String|EqualExactly(String) (Variables|Default|GetCurrentWeek) (Utilities|Struct|BreakCalendarDataLayerData _array_element))
      (bind _returnvalue (WorldPartition|GetDataLayerManager))
      (Utilities|FlowControl|Sequence
        (:then_0
          (for _array_element_2 (.EnabledDataLayers_9_41BB9A3E4648805B5AF70AA728288C92 (Utilities|Struct|BreakCalendarDataLayerData _array_element))
            (DataLayers|SetDataLayerInstanceRuntimeState _returnvalue (DataLayers|GetDataLayerInstancefromAsset _returnvalue _array_element_2) "Activated" false)))
        (:then_1
          (for _array_element_3 (.DisabledDataLayers_10_8DE59E3B474C88ED2E7988B185572444 (Utilities|Struct|BreakCalendarDataLayerData _array_element))
            (DataLayers|SetDataLayerInstanceRuntimeState _returnvalue (DataLayers|GetDataLayerInstancefromAsset _returnvalue _array_element_3) "Unloaded" false)))))))

(fn ConstructionScript
  (|CallParentFunction))
